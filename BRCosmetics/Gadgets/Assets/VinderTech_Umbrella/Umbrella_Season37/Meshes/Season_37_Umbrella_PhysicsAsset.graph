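# Physics asset for the Season 37 VinderTech Umbrella gadget mesh, providing simplified collision bodies and joints for its skeleton.
# Three capsules cover the bones: a long, narrow one for the two-handed melee handle and broad ones for the blades and spinning top masters.
# Blades and top each constrain directly to the handle in a star layout, with matching swing and twist limit angles.
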PhysicsAsset /BRCosmetics/Gadgets/Assets/VinderTech_Umbrella/Umbrella_Season37/Meshes/Season_37_Umbrella_PhysicsAsset
bodies {
 Melee_TwoHanded_Handle [{shapeType=Capsule, length=60.5864, radius=7.8776, center=<0 0.6205 42.7608>, rotation=<rot -0.6092 -90 0>}]
 blades_mstr [{shapeType=Capsule, length=66.6492, radius=66.6492, center=<0.0002 0.0001 0.3867>, rotation=<rot 0 -58.3807 -89.9954>}]
 spin_top_mstr [{shapeType=Capsule, length=43.0036, radius=38.6343, center=<0.1958 0.0818 -0.1612>, rotation=<rot 0 -67.3381 -88.6374>}]
}
constraints [
 {bone1Name=spin_top_mstr, bone2Name=Melee_TwoHanded_Handle, swing1LimitDegrees=45, swing2LimitDegrees=45, twistLimitDegrees=45}
 {bone1Name=blades_mstr, bone2Name=Melee_TwoHanded_Handle, swing1LimitDegrees=45, swing2LimitDegrees=45, twistLimitDegrees=45}
]
